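# Parent character material for the Beagle skin of the male LivelyDomino body, layering emissive skin effects onto the shared 2023 character material.
# PreFX output feeds a LivelyDomino advanced emissive function, then an optional, static-switched panning emissive using selectable Tech Art Mask channels.
# The resulting emissive color passes through a time-of-day custom colors function before the material enters the shared PostFX function.
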Material /BRCosmetics/Characters/Player/Male/Medium/Bodies/M_MED_LivelyDomino/Skins/Beagle/Materials/M_LivelyDomino_Beagle_Parent
graph {
 properties {
  bUseMaterialAttributes=true
  bUsedWithMeshDeformer=true
  bUsedWithMorphTargets=true
  bUsedWithSkeletalMesh=true
  pixelDepthOffsetMode=PDOM_Legacy
  shadingModel=MSM_FromMaterialExpression
  usedShadingModels=Default Lit | Subsurface | Clear Coat | Subsurface Profile | Cloth
 }
 nodes [
  n1 = MaterialFunctionCall() {materialFunction=&MF_CharacterParent_2023_PreFX}
  n2 = StaticComponentMaskParameter(None=@n1.Tech Art Mask) {defaultR=true, group=Skin FX, parameterName=PanningEmissive_Channel}
  n3 = StaticComponentMaskParameter(None=@n1.Tech Art Mask) {defaultB=true, group=Skin FX, parameterName=PanningEmissive_UVOffset_Channel}
  n4 = MaterialFunctionCall!(Local Position, Alpha=@n1.Tech Art Mask, Fresnel, Material=@n1.Tech Art Mask, Local Normal) {materialFunction=&/MF_AdvancedEmissive_LivelyDomino}
  n5 = GetMaterialAttributes(None=StaticSwitchParameter(
    True=MaterialFunctionCall!(Alpha=@n2, isVehicle, Gmap, UVOffset=@n3, Local Normal, Local Position, EmissiveColorOverride, Material=@n4.Result) {materialFunction=/Game/Weapons/WeaponSkins/MaterialFunction/PanningEmissive}.Result
    False=@n4.Result
   ) {group=Skin FX, parameterName=Use Panning Emissive}) {attributeGetTypes [#1]}
  MaterialFunctionCall(
   Material=SetMaterialAttributes(
    MaterialAttributes=@n5.MaterialAttributes
    Emissive Color=MaterialFunctionCall!(In=@n5.Emissive Color) {materialFunction=/Game/Packages/Fortress_SharedMaterials/Anime_Master_Material/MF_TODCustomColors}.Result
   ) {attributeSetTypes [#1]}
   M=@n1.M
   Pre Skinned Local Position=@n1.M
   Distance Blend=@n1.M
   Fresnel=@n1.M
  ) {materialFunction=&MF_CharacterParent_2023_PostFX}
 ]
}
